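; Shared Stamina weapon Blueprint for the Hawaii item, a limited-use charged beam that relays ability stages through dispatchers.
; Equip sends a gameplay event and defers OnEquipped until initial ammo is known; unequip removes the equipped effect.
; Trigger presses activate the primary ability only from permitted skydiving, while targeting toggles a camera modifier off-server.
; Charge start consumes ammo, and at zero a delayed cleanup destroys the instigator's Hawaii component and removes the item.
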(event EventBeginPlay
  ;[1 1 1] BeginPlay
  (EventDispatchers|BindEventtoOnLocalAmmoChanged (delegate Custom|Ammochanged))
  (EventDispatchers|BindEventtoOnPressTrigger (delegate Custom|PressedTrigger)))

(event Custom|Ammochanged
  (params (LocalCount :type Integer) (LocalRemaining :type Integer))
  ;[1 1 1] BeginPlay
  (Variables|Default|SetLocalBPAmmo LocalRemaining)
  (if (not (Variables|Default|GetGotInitialAmmo))
    (Variables|Default|SetGotInitialAmmo true)
    (Default|CallOnEquipped)))

(event Custom|PressedTrigger
  (params (Weapon :type Fort%20Weapon%20Object%20Reference))
  ;[1 1 1] BeginPlay
  (bind _asfort_player_pawn (Utilities|Casting|CastToFortPlayerPawn (Game|GetInstigator)))
  (if (or (Health|IsDBNO _asfort_player_pawn) (FortPawn|IsDead _asfort_player_pawn))
    _
    (elif (CallFunction|CanTriggerfromSkydiving _asfort_player_pawn)
      (Ability|SendGameplayEventtoActor (Game|GetInstigator) "(TagName=\"Stamina.Hawaii.Event.TriggerFromAlaska\")")
      (Abilities|TryActivateAbilitybyClass (Ability|GetAbilitySystemComponent _asfort_player_pawn) "/StaminaGameplay/Gameplay/Items/Hawaii/GA_Stamina_Hawaii_Primary.GA_Stamina_Hawaii_Primary_C" true)
      (else
        (if (and (GameplayTags|HasMatchingGameplayTag _asfort_player_pawn (Variables|Default|GetSprintTag)) (GameplayTags|HasMatchingGameplayTag _asfort_player_pawn (Variables|Default|GetFallingTag)))
          _)))))

(event FortWeapon|EventOnEquippedAndAttached
  ;[1 1 1] Equip
  (Ability|SendGameplayEventtoActor (Game|GetInstigator) "(TagName=\"Stamina.Hawaii.Event.Equipped\")")
  (if (Variables|Default|GetGotInitialAmmo)
    (Variables|Default|SetLocalBPAmmo (FortWeapon|GetLocalMagazineAmmoCount))
    (Default|CallOnEquipped)
    (Utilities|Casting|CastToFortPlayerPawn (Game|GetInstigator))))

(event FortWeapon|EventOnUnEquippedAndDetached
  ;[1 1 1] UnEquip
  (bind _asfort_player_pawn (Utilities|Casting|CastToFortPlayerPawn (Game|GetInstigator)))
  (GameplayEffects|RemoveActiveEffectswithGrantedTags (Ability|GetAbilitySystemComponent _asfort_player_pawn) "(GameplayTags=((TagName=\"Stamina.Hawaii.Equipped\")))"))

(event Custom|SetCameraTargeting
  (params (InTargeting :type Boolean))
  ;[1 1 1] Set Targeting
  (if (Networking|IsServer)
    _
    (elif InTargeting
      (bind _outmodifier (Class|WeaponFunctionLibrary|FindorAddCameraModifier "/StaminaGameplay/Gameplay/Items/Hawaii/Stamina_Hawaii_CameraModifier_Targeting.Stamina_Hawaii_CameraModifier_Targeting_C" self))
      (if true
        (CameraModifier|EnableModifier _outmodifier))
      (else
        (bind _outmodifier_2 (Class|WeaponFunctionLibrary|FindorAddCameraModifier "/StaminaGameplay/Gameplay/Items/Hawaii/Stamina_Hawaii_CameraModifier_Targeting.Stamina_Hawaii_CameraModifier_Targeting_C" self))
        (CameraModifier|DisableModifier _outmodifier_2 false)))))

(fn AbilityEnd
  (Default|CallAbilityFinished))

(fn BeamFired
  (Default|CallBeamFiredDelegate))

(fn CanTriggerFromSkydiving
  (params (Pawn :type Fort%20Player%20Pawn%20Object%20Reference))
  (LocalVariables|SetHasAlaskaTag (GameplayTags|HasMatchingGameplayTag Pawn (Variables|Default|GetRequiredTag)))
  (bind _output_get (LocalVariables|SetIsSkydiving (Parachute|IsSkydiving Pawn)))
  (if _output_get
    (return (or (LocalVariables|GetHasAlaskaTag) (ScalableFloat|GetValueasBool (Variables|Default|GetRow_AllowAnyGlideCancel))))
    (else
      (return false))))

(fn ChargeStart
  (if (not (FortWeapon|Ammo|HasInfiniteUses))
    (FortWeapon|Ammo|ConsumeAmmo 1))
  (Default|CallChargeStarted)
  (if (== (FortWeapon|GetRemainingAmmo) 0)
    (Utilities|Time|SetTimerbyEvent (EventDispatchers|CreateEvent :!event DestroySelf) 4.8)))

(fn DestroySelf
  (bind _returnvalue (Game|GetInstigatorController))
  (bind _instigator (Variables|Actor|GetInstigator))
  (Utilities|IsValid _instigator
    (:Is%20Valid
      (Components|DestroyComponent (Actor|GetComponentbyClass _instigator "/StaminaGameplay/Gameplay/Items/Hawaii/Stamina_Hawaii_Component.Stamina_Hawaii_Component_C"))
      (Utilities|IsValid _returnvalue
        (:Is%20Valid
          (Inventory|RemoveItemFromPlayerByGuid _returnvalue (FortWeapon|GetInventoryGUID) 1 true))))))

(fn ConstructionScript)
