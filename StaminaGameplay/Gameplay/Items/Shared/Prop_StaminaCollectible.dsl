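; Creative coin-style collectible prop for Stamina capsules and Dragon Balls that reports pickups to quest, minigame stat and HUD score systems.
; SendStatEvents resolves the player controller from the player state and adds the current coin tag to the target tags.
; It sends a capsule or ball quest event by substring-matching the visible mesh's name ("Container" or "ball").
; It then sends Collect and score stat events and broadcasts the score to the HUD.
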(fn SendStatEvents
  (params (InFortPlayerState :type Fort%20Player%20State%20Object%20Reference))
  (bind (_outsourcetags _outcontexttags) (Class|FortMinigameStatsComponent|GetSourceandContextTags (LocalVariables|Getlocal%20MinigameStatsComp) (Variables|Default|GetForPlayerController_0)))
  (bind _visiblemesh (Variables|Default|GetVisibleMesh))
  (bind _returnvalue (Utilities|GetDisplayName _visiblemesh))
  (bind _asfort_player_controller_zone (Utilities|Casting|CastToFortPlayerControllerZone (Actor|GetOwner InFortPlayerState)))
  (Variables|Default|SetForPlayerController_0 _asfort_player_controller_zone)
  (bind _output_get (Variables|Default|SetTargetTags_0 (GameplayTags|GetOwnedGameplayTags self :!class /Script/GameplayTags.GameplayTagAssetInterface)))
  (GameplayTags|AddGameplayTag _output_get (Variables|Default|GetCurrentCoinTag) :!class /Script/GameplayTags.BlueprintGameplayTagLibrary)
  (Utilities|FlowControl|Sequence
    (:then_0
      (bind _output_get_2 (Variables|Default|SetQuestManager_0 (Quests|GetQuestManager (Variables|Default|GetForPlayerController_0) "Athena")))
      (Utilities|IsValid _output_get_2
        (:Is%20Valid
          ;[1 1 1] Send appropriate quest events if Capsule or Dragonball is picked up
          (if (Utilities|String|Contains _returnvalue "Container")
            (Quest|SendComplexCustomStatEvent (Variables|Default|GetQuestManager_0) self "(GameplayTags=((TagName=\"Creative.Item.Coin.StaminaCapsule\")))" (Variables|Default|GetTargetTags_0) (Variables|Default|GetAmountToCount))
            (elif (Utilities|String|Contains _returnvalue "ball")
              (Quest|SendComplexCustomStatEvent (Variables|Default|GetQuestManager_0) self "(GameplayTags=((TagName=\"Creative.Item.Coin.StaminaBall\")))" (Variables|Default|GetTargetTags_0) (Variables|Default|GetAmountToCount)))))))
    (:then_1
      (bind _returnvalue_2 (CreativeDevices|GetMinigameStatsComp (Variables|Default|GetForPlayerController_0) "Athena"))
      (LocalVariables|Setlocal%20MinigameStatsComp _returnvalue_2)
      (bind _local_minigamestatscomp (LocalVariables|Getlocal%20MinigameStatsComp)
        (:then
          (Class|FortMinigameStatsComponent|SendStatEventwithTags _local_minigamestatscomp (Variables|Default|GetForPlayerController_0) "Collect" self (Variables|Default|GetTargetTags_0) (Class|FortMinigameStatsComponent|GetSourceandContextTags _local_minigamestatscomp (Variables|Default|GetForPlayerController_0)) (.OutContextTags (Class|FortMinigameStatsComponent|GetSourceandContextTags _local_minigamestatscomp (Variables|Default|GetForPlayerController_0))) (Variables|Default|GetAmountToCount))
          (GameplayTags|AddGameplayTag (Variables|Default|GetTargetTags_0) "(TagName=\"Minigame.Stat.Score.Coin\")" :!class /Script/GameplayTags.BlueprintGameplayTagLibrary)
          (Class|FortMinigameStatsComponent|SendStatEventwithTags (LocalVariables|Getlocal%20MinigameStatsComp) (Variables|Default|GetForPlayerController_0) "MinigameDynamicEvent" self (Variables|Default|GetTargetTags_0) _outsourcetags _outcontexttags (Variables|Default|GetScoreValue))
          (Class|CreativeHUDBroadcastComponent|DoScoreEvent (Variables|Default|GetCreative_HUDBroadcast_Component) (Variables|Default|GetForPlayerController_0) (Variables|Default|GetScoreValue) (Variables|UserOptions|Advanced|GetDisplayScoreUpdateOnHUD)))
        (:else)))))

(fn ConstructionScript
  (|CallParentFunction))
